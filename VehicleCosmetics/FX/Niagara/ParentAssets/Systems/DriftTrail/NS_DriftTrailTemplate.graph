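# NS_DriftTrailTemplate is a parent Niagara system template for vehicle cosmetic drift trails.
# It has no emitters, only system-level SystemState and a variable assignment module, plus a skeletal mesh data interface.
# User parameters define the shared inputs: frontend preview flag, painted flag, color and scalar, color variant ID, wheel size and a wheel offset set through the cosmetic template.
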NiagaraSystem /VehicleCosmetics/FX/Niagara/ParentAssets/Systems/DriftTrail/NS_DriftTrailTemplate
dependencies {
 usedDataInterfaces [&NiagaraDataInterfaceSkeletalMesh]
 usedDynamicInputs []
 usedModules [
  &self:SystemSpawnScript.SystemScriptSource.SystemScriptGraph.NiagaraNodeAssignment_0.SetVariables_#1
  &SystemState
 ]
}
summary {
 systemName=NS_DriftTrailTemplate
 emitters []
 userVariables [
  User.bIsFrontendPreview {type=NiagaraBool}
  User.bIsPainted {type=NiagaraBool}
  User.ColorVariantID {type=NiagaraInt32}
  User.offset {type=Vector3f, description="offset from wheels(set through cosmetic template)", defaultValue {struct=Vector3f, value=<0 0 0>}}
  User.PaintedColor {type=LinearColor, description="", defaultValue {struct=LinearColor, value=<rgba 1 1 1 1>}}
  User.PaintedScalar {type=NiagaraFloat, defaultValue=1, description=""}
  User.WheelSize {type=NiagaraFloat, defaultValue=0.1, description=""}
 ]
}
